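# Unlit material for Niagara mesh particles that draws flowers in the main player's Jovial emote.
# Particle color, varied in brightness by clamped per-particle random, blends toward a warm red-orange tint using the petal texture as a mask.
# Vertex color red then blends in a randomly dimmed yellow, likely for flower centers.
# The result feeds the shared MF_FX_EmissiveAdjustments function as emissive, and ray-traced shadows are off.
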Material /Game/Animation/Game/MainPlayer/Emotes/Jovial/FX/M_Emote_Jovial_Flowers
graph {
 properties {
  bCastRayTracedShadows=false
  bUsedWithNiagaraMeshParticles=true
  pixelDepthOffsetMode=PDOM_Legacy
  shadingModel=MSM_Unlit
  usedShadingModels=Unlit
 }
 nodes [
  Constant3Vector() {constant=<rgba 1 1 1 1>}
  Constant()
  n1 = Multiply(A=ParticleColor().RGB, B=Clamp(None=ParticleRandom(), Min=ScalarParameter() {defaultValue=0.5271}) {minDefault=0.3})
  n2 = VectorParameter() {defaultValue=<rgba 1 0.1417 0.0607 1>}
  n3 = TextureSample(UVs=TextureCoordinate()) {texture=&/T_Emote_Jovial_Petal}
  n4 = VectorParameter() {parameterName=Param_1, defaultValue=<rgba 1 0.8818 0.0868 1>}
  n5 = Clamp(None=ParticleRandom()) {minDefault=0.6}
  MaterialFunctionCall(Emissive=LinearInterpolate(A=LinearInterpolate(A=@n1, B=@n2.RGB, Alpha=@n3.RGB), B=Multiply(A=@n4.RGB, B=@n5), Alpha=VertexColor().R)) {materialFunction=&MF_FX_EmissiveAdjustments}
 ]
}
