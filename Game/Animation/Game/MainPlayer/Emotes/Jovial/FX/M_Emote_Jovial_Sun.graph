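# Translucent, unlit, two-sided material for the sun effect in the Jovial player emote, usable on Niagara sprites and mesh particles.
# A Fresnel or switchable UV sphere mask drives opacity, combined with a panning Power Source noise texture, and blends the two-color textured emissive toward a hot tint.
# The UseEyes? parameter lerps both outputs to an eyes texture, whose red channel takes particle color and green channel supplies orange fill and opacity.
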Material /Game/Animation/Game/MainPlayer/Emotes/Jovial/FX/M_Emote_Jovial_Sun
graph {
 properties {
  bOutputTranslucentVelocity=true
  bUsedWithNiagaraMeshParticles=true
  bUsedWithNiagaraSprites=true
  blendMode=BLEND_Translucent
  pixelDepthOffsetMode=PDOM_Legacy
  shadingModel=MSM_Unlit
  twoSided=true
  usedShadingModels=Unlit
 }
 nodes [
  n1 = TextureCoordinate()
  n2 = Constant2Vector() {g=0.5, r=0.5}
  n3 = ScalarParameter() {defaultValue=0.75, group=Sphere Mask, parameterName=Sphere Mask Radius}
  n4 = ScalarParameter() {group=Sphere Mask, parameterName=Sphere Mask Hardness}
  n5 = OneMinus(None=Power(
    Base=SphereMask(A=@n1, B=@n2, Radius=@n3, Hardness=@n4) {hardnessPercent=0}
    Exp=ScalarParameter() {defaultValue=1, group=Sphere Mask, parameterName=Sphere Mask Power}
   ) {constExponent=1})
  n6 = If(
   A=ScalarParameter() {group=Sphere Mask, parameterName=Use Sphere Mask instead of Fresnel}
   "A > B"=@n5
   "A == B"=@n5
   "A < B"=Fresnel(ExponentIn=ScalarParameter() {defaultValue=0.4, parameterName=Param_1}) {exponent=0.3}
  ) {constB=0.99}
  n7 = OneMinus(None=MaterialFunctionCall(In=@n6, Contrast=ScalarParameter() {defaultValue=1, parameterName=Opacity Contrast}) {materialFunction=&CheapContrast}.Result)
  n8 = TextureCoordinate()
  n9 = AppendVector(A=Constant() {r=1.5}, B=Constant() {r=1.286})
  n10 = Constant() {r=1.5}
  n11 = Constant()
  n12 = TextureSample(UVs=Panner(Coordinate=Multiply(A=@n8, B=@n9), Speed=AppendVector(A=@n10, B=@n11))) {texture=/Game/Effects/Fort_Effects/Textures/Power_Source/T_Power_Source_Sphere_Texture}
  n13 = TextureSample() {texture=&/T_Emote_Jovial_Eyes}
  LinearInterpolate(A=Saturate(None=Add(A=@n7, B=@n12.RGB)), B=Add(A=@n13.G, B=@n13.G), Alpha=ScalarParameter() {parameterName=UseEyes?})
  n14 = ScalarParameter() {defaultValue=1.12, parameterName=X}
  n15 = ScalarParameter() {defaultValue=1.7367, parameterName=Y}
  n16 = TextureSample(UVs=Multiply(A=TextureCoordinate(), B=AppendVector(A=@n14, B=@n15))) {texture=&T_Power_Source_Sphere_Texture}
  n17 = ScalarParameter() {defaultValue=0.1, parameterName=Contrast}
  n18 = MaterialFunctionCall(
   LDR Input=LinearInterpolate(
    A=VectorParameter() {parameterName=ColorA, defaultValue=<rgba 1 0.698 0 1>}.RGB
    B=VectorParameter() {parameterName=ColorB, defaultValue=<rgba 0.761 0.2421 0 1>}.RGB
    Alpha=MaterialFunctionCall(In=@n16.RGB, Contrast=@n17) {materialFunction=&CheapContrast}.Result
   )
   HDR Intensity Multiplier=Constant() {r=1}
  ) {materialFunction=&DeriveHDRfromLDR}
  n19 = VectorParameter() {defaultValue=<rgba 2 1.6008 0.678 1>}
  n20 = MaterialFunctionCall(In=@n6, Contrast=ScalarParameter() {defaultValue=1, parameterName=Emissive Contrast}) {materialFunction=&CheapContrast}
  n21 = TextureSample() {texture=&T_Emote_Jovial_Eyes}
  n22 = Multiply(A=@n21.R, B=ParticleColor().RGB)
  n23 = Constant3Vector() {constant=<rgba 1 0.3813 0 1>}
  MaterialFunctionCall(Emissive=LinearInterpolate(
    A=LinearInterpolate(A=@n18.RemappedRange, B=@n19.RGB, Alpha=@n20.Result)
    B=LinearInterpolate(A=@n22, B=@n23, Alpha=@n21.G)
    Alpha=ScalarParameter() {parameterName=UseEyes?}
   )) {materialFunction=&MF_FX_EmissiveAdjustments}
 ]
}
