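# Translucent unlit Niagara sprite material that draws sun rays for the player's Jovial emote FX.
# A ray texture, masked by a radial falloff shaped through the Scale_RayBeams dynamic parameter, joins a soft core glow, with the center cut out.
# Particle color tints that shape for emissive.
# Opacity subtracts clamped, rotating cloud noise from the shape, then applies a contrast-adjusted vertical gradient, depth fade and particle alpha.
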Material /Game/Animation/Game/MainPlayer/Emotes/Jovial/FX/M_Emote_Jovial_Sun_Rays
graph {
 properties {bUsedWithNiagaraSprites=true, blendMode=BLEND_Translucent, pixelDepthOffsetMode=PDOM_Legacy, shadingModel=MSM_Unlit, usedShadingModels=Unlit}
 nodes [
  Constant() {r=-3}
  Constant() {r=2}
  n1 = TextureCoordinate()
  MaterialFunctionCall(UVs=@n1, "Rotation Angle (0-1)"=Multiply(A=Time(), B=Constant() {r=0.1})) {materialFunction=&CustomRotator}
  MaterialFunctionCall!(Hardness, Center, Radius=Constant() {r=0.25}) {materialFunction=/Game/Packages/Fortress_SharedMaterials/Base_Material_Functions/FX_SphereMask2D}
  n2 = MaterialFunctionCall(Radius=ScalarParameter() {defaultValue=0.3, parameterName=Param_2}) {materialFunction=&FX_SphereMask2D}
  n3 = TextureSample(UVs=@n1) {samplerType=SAMPLERTYPE_LinearColor, texture=&/T_Emote_Jovial_SunRays}
  n4 = Power(
   Base=MaterialFunctionCall() {materialFunction=&FX_SphereMask2D}.Result
   Exp=DynamicParameter() {paramNames [Scale_RayBeams, Param2, Param3, Param4]}.Scale_RayBeams
  ) {constExponent=1}
  n5 = ScalarParameter() {defaultValue=0.14, parameterName=Param_1}
  n6 = Multiply(
   A=Add(A=Power(Base=@n2.Result) {constExponent=3}, B=Multiply(A=@n3.R, B=@n4))
   B=OneMinus(None=MaterialFunctionCall(Radius=@n5) {materialFunction=&FX_SphereMask2D}.Result)
  )
  MaterialFunctionCall(Emissive=Multiply(A=ParticleColor().RGB, B=@n6)) {materialFunction=&MF_FX_EmissiveAdjustments}
  n7 = OneMinus(None=MaterialFunctionCall(Radius=Constant() {r=0.3}) {materialFunction=&FX_SphereMask2D}.Result)
  n8 = Time()
  n9 = Constant() {r=0.2}
  n10 = TextureSample(UVs=MaterialFunctionCall(UVs=TextureCoordinate(), "Rotation Angle (0-1)"=Multiply(A=@n8, B=@n9)) {materialFunction=&CustomRotator}.Rotated Values) {texture=/EpicBaseTextures/Noises/T_Clouds_Noise3_D}
  n11 = Subtract(A=@n6, B=Clamp(None=Multiply(A=@n7, B=@n10.RGB)) {maxDefault=0.2, minDefault=0.01})
  n12 = TextureCoordinate()
  n13 = MaterialFunctionCall(
   In=Power(Base=ComponentMask(None=@n12) {g=true}, Exp=ScalarParameter() {defaultValue=0.6, parameterName=ss})
   Contrast=ScalarParameter() {defaultValue=0.5, parameterName=s}
  ) {materialFunction=&CheapContrast}
  Multiply(
   A=ParticleColor().A
   B=DepthFade(Opacity=Multiply(A=@n11, B=@n13.Result), FadeDistance=ScalarParameter() {defaultValue=11, parameterName=Param_3})
  )
 ]
}
